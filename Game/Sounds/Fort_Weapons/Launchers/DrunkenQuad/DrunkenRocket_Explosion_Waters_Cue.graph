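# Sound cue for the DrunkenQuad launcher's rocket explosion in water, layered by distance.
# Each mix blends close, medium and distant layers, each with its own attenuation and random wave pick; the close layer adds a boosted water debris wave, and the others vary pitch.
# A quality node gives its first two levels four variations per layer and its last two levels two.
# Concurrency rules cap simultaneous explosions, stopping the oldest.
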SoundCue /Game/Sounds/Fort_Weapons/Launchers/DrunkenQuad/DrunkenRocket_Explosion_Waters_Cue
root=QualityLevel_0
nodes [
 {
  id=Attenuation_0
  children [Modulator_1]
  properties {
   attenuationSettings=/Game/Sounds/Attenuation/Fort_Explosion_Distant_Athena_Att
   attenuationOverrides {falloffDistance=3000, attenuationShapeExtents {x=1000}}
  }
 }
 {
  id=Attenuation_1
  children [Mixer_2]
  properties {attenuationSettings=/Game/Sounds/Attenuation/Fort_Explosion_Close_Athena_Att, attenuationOverrides {falloffDistance=1600}}
 }
 {
  id=Attenuation_2
  children [Modulator_2]
  properties {
   attenuationSettings=/Game/Sounds/Attenuation/Fort_Explosion_Medium_Athena_Att
   attenuationOverrides {falloffDistance=3000, attenuationShapeExtents {x=1000}}
  }
 }
 {
  id=Attenuation_3
  children [Modulator_4]
  properties {
   attenuationSettings=&Fort_Explosion_Distant_Athena_Att
   attenuationOverrides {falloffDistance=3000, attenuationShapeExtents {x=1000}}
  }
 }
 {
  id=Attenuation_4
  children [Mixer_1]
  properties {attenuationSettings=&Fort_Explosion_Close_Athena_Att, attenuationOverrides {falloffDistance=1600}}
 }
 {
  id=Attenuation_5
  children [Modulator_5]
  properties {
   attenuationSettings=&Fort_Explosion_Medium_Athena_Att
   attenuationOverrides {falloffDistance=3000, attenuationShapeExtents {x=1000}}
  }
 }
 {id=Mixer_0, children [SoundClass_4,SoundClass_1,SoundClass_5], properties {inputVolume [1,1,1]}}
 {id=Mixer_1, children [Random_3,WavePlayer_0], properties {inputVolume [1,2.3]}}
 {id=Mixer_2, children [Random_6, WavePlayer_13], properties {inputVolume [1,2.3]}}
 {id=Mixer_4, children [SoundClass_3,SoundClass_2,SoundClass_0], properties {inputVolume [1,1,1]}}
 {id=Modulator_0, children [Attenuation_4]}
 {id=Modulator_1, children [Random_5], properties {pitchMax=1.1, pitchMin=0.9}}
 {id=Modulator_2, children [Random_7], properties {pitchMax=1.1, pitchMin=0.9}}
 {id=Modulator_3, children [Attenuation_1]}
 {id=Modulator_4, children [Random_8], properties {pitchMax=1.1, pitchMin=0.9}}
 {id=Modulator_5, children [Random_4], properties {pitchMax=1.1, pitchMin=0.9}}
 {id=QualityLevel_0, children [Mixer_4,Mixer_4,Mixer_0,Mixer_0]}
 {id=Random_3, children [WavePlayer_2,WavePlayer_3,WavePlayer_5,WavePlayer_6], properties {weights [1,1,1,1]}}
 {id=Random_4, children [WavePlayer_7, WavePlayer_8, WavePlayer_9, WavePlayer_10], properties {weights [1,1,1,1]}}
 {id=Random_5, children [WavePlayer_11, WavePlayer_18, WavePlayer_19, WavePlayer_20], properties {weights [1,1,1,1]}}
 {id=Random_6, children [WavePlayer_21, WavePlayer_22], properties {weights [1,1]}}
 {id=Random_7, children [WavePlayer_25, WavePlayer_26], properties {weights [1,1]}}
 {id=Random_8, children [WavePlayer_29, WavePlayer_30], properties {weights [1,1]}}
 {
  id=SoundClass_0
  children [Attenuation_0]
  properties {soundClassOverride=&Fort_WeaponRanged_Fire_Distant}
 }
 {
  id=SoundClass_1
  children [Attenuation_2]
  properties {soundClassOverride=&Fort_WeaponRanged_Fire_Distant}
 }
 {
  id=SoundClass_2
  children [Attenuation_5]
  properties {soundClassOverride=&Fort_WeaponRanged_Fire_Distant}
 }
 {id=SoundClass_3, children [Modulator_0], properties {soundClassOverride=&Fort_WeaponRanged_Fire_3P_PSM}}
 {id=SoundClass_4, children [Modulator_3], properties {soundClassOverride=&Fort_WeaponRanged_Fire_3P_PSM}}
 {
  id=SoundClass_5
  children [Attenuation_3]
  properties {soundClassOverride=&Fort_WeaponRanged_Fire_Distant}
 }
 {id=WavePlayer_0, properties {soundWaveAssetPtr=/Game/Athena/Sounds/Weapons/Explosions/EXPL_Debris_Water_Wave_01}}
 {id=WavePlayer_10, properties {soundWaveAssetPtr=&/DrunkenRocket_Explosion_Medium_04}}
 {id=WavePlayer_11, properties {soundWaveAssetPtr=&/DrunkenRocket_Explosion_Distant_01}}
 {id=WavePlayer_13, properties {soundWaveAssetPtr=&EXPL_Debris_Water_Wave_01}}
 {id=WavePlayer_18, properties {soundWaveAssetPtr=&/DrunkenRocket_Explosion_Distant_02}}
 {id=WavePlayer_19, properties {soundWaveAssetPtr=&/DrunkenRocket_Explosion_Distant_03}}
 {id=WavePlayer_2, properties {soundWaveAssetPtr=&/DrunkenRocket_Explosion_Close_01}}
 {id=WavePlayer_20, properties {soundWaveAssetPtr=&/DrunkenRocket_Explosion_Distant_04}}
 {id=WavePlayer_21, properties {soundWaveAssetPtr=&DrunkenRocket_Explosion_Close_01}}
 {id=WavePlayer_22, properties {soundWaveAssetPtr=&/DrunkenRocket_Explosion_Close_02}}
 {id=WavePlayer_25, properties {soundWaveAssetPtr=&/DrunkenRocket_Explosion_Medium_01}}
 {id=WavePlayer_26, properties {soundWaveAssetPtr=&/DrunkenRocket_Explosion_Medium_02}}
 {id=WavePlayer_29, properties {soundWaveAssetPtr=&DrunkenRocket_Explosion_Distant_01}}
 {id=WavePlayer_3, properties {soundWaveAssetPtr=&DrunkenRocket_Explosion_Close_02}}
 {id=WavePlayer_30, properties {soundWaveAssetPtr=&DrunkenRocket_Explosion_Distant_02}}
 {id=WavePlayer_5, properties {soundWaveAssetPtr=&/DrunkenRocket_Explosion_Close_03}}
 {id=WavePlayer_6, properties {soundWaveAssetPtr=&/DrunkenRocket_Explosion_Close_04}}
 {id=WavePlayer_7, properties {soundWaveAssetPtr=&DrunkenRocket_Explosion_Medium_01}}
 {id=WavePlayer_8, properties {soundWaveAssetPtr=&DrunkenRocket_Explosion_Medium_02}}
 {id=WavePlayer_9, properties {soundWaveAssetPtr=&/DrunkenRocket_Explosion_Medium_03}}
]
properties {
 maxDistance=27500
 volumeMultiplier=0.9
 concurrencyOverrides {maxCount=4, resolutionRule=StopOldest, volumeScale=0.5}
 concurrencySet [/Game/Sounds/Concurrency_Rules/Weapon_Explosion_Conc]
}
