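# First-person fire sound cue for the DrunkenQuad rocket launcher.
# A Random node picks one of four equally weighted close-fire waves, feeding a Modulator that slightly varies and boosts volume while leaving pitch fixed.
# The cue routes to the ranged-weapon first-person fire sound class and uses shared first-person projectile attenuation and gunfire concurrency assets.
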SoundCue /Game/Sounds/Fort_Weapons/Launchers/DrunkenQuad/DrunkenRocket_Fire_1P_Cue
root=Modulator_0
nodes [
 {id=Modulator_0, children [Random_1], properties {pitchMax=1, pitchMin=1, volumeMax=1.2, volumeMin=1.1}}
 {id=Random_1, children [WavePlayer_0,WavePlayer_5,WavePlayer_6,WavePlayer_7], properties {weights [1,1,1,1]}}
 {id=WavePlayer_0, properties {soundWaveAssetPtr=&/DrunkenRocket_Fire_Close_01}}
 {id=WavePlayer_5, properties {soundWaveAssetPtr=&/DrunkenRocket_Fire_Close_02}}
 {id=WavePlayer_6, properties {soundWaveAssetPtr=&/DrunkenRocket_Fire_Close_03}}
 {id=WavePlayer_7, properties {soundWaveAssetPtr=&/DrunkenRocket_Fire_Close_04}}
]
properties {
 attenuationSettings=&Fort_Weapon_Projectile_FirstPerson_Athena_Att
 maxDistance=2097152
 soundClassObject=&Fort_WeaponRanged_Fire_1P_PSM
 volumeMultiplier=0.85
 attenuationOverrides {falloffDistance=5000, attenuationShapeExtents {x=1000}}
 concurrencyOverrides {bLimitToOwner=true, maxCount=1, resolutionRule=StopOldest}
 concurrencySet [&Weapon_Gunfire_FirstPerson_Conc]
}
