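# Sound cue for firing the DrunkenQuad rocket launcher, layering close, medium and distant shots by distance.
# A quality-level switch feeds two mixers: the first two slots pick from four random waves per layer, the last two from two.
# Each layer adds random volume or pitch modulation, its own attenuation and a sound class override (third-person fire or distant fire).
# Local and global gunfire concurrency sets limit simultaneous plays.
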SoundCue /Game/Sounds/Fort_Weapons/Launchers/DrunkenQuad/DrunkenRocket_Fire_Cue
root=QualityLevel_0
nodes [
 {id=Attenuation_0, children [Modulator_1], properties {attenuationSettings=&Fort_Weapon_Projectile_Close_Athena_Att}}
 {
  id=Attenuation_1
  children [Modulator_3]
  properties {attenuationSettings=&Fort_Weapon_Projectile_Medium_Mono_Athena_Att}
 }
 {
  id=Attenuation_10
  children [Modulator_0]
  properties {attenuationSettings=&Fort_Weapon_Projectile_Medium_Mono_Athena_Att}
 }
 {
  id=Attenuation_11
  children [Modulator_7]
  properties {attenuationSettings=&Fort_Weapon_Projectile_Distant_Mono_Athena_Att}
 }
 {
  id=Attenuation_2
  children [Modulator_2]
  properties {attenuationSettings=&Fort_Weapon_Projectile_Distant_Mono_Athena_Att}
 }
 {id=Attenuation_9, children [Modulator_6], properties {attenuationSettings=&Fort_Weapon_Projectile_Close_Athena_Att}}
 {id=Mixer_0, children [SoundClass_1,SoundClass_2,SoundClass_5], properties {inputVolume [1,0.9,0.8]}}
 {id=Mixer_3, children [SoundClass_7,SoundClass_4,SoundClass_3], properties {inputVolume [1,0.9,0.8]}}
 {id=Modulator_0, children [Random_2], properties {volumeMax=1.95, volumeMin=1.85}}
 {id=Modulator_1, children [Random_6], properties {pitchMax=1.1, pitchMin=0.9, volumeMax=0.95, volumeMin=0.85}}
 {id=Modulator_2, children [Random_8], properties {pitchMax=0.9, pitchMin=0.8, volumeMax=1.6, volumeMin=1.5}}
 {id=Modulator_3, children [Random_7], properties {volumeMax=1.95, volumeMin=1.85}}
 {id=Modulator_6, children [Random_1], properties {volumeMax=0.95, volumeMin=0.85}}
 {id=Modulator_7, children [Random_5], properties {pitchMax=0.9, pitchMin=0.8, volumeMax=1.6, volumeMin=1.5}}
 {id=QualityLevel_0, children [Mixer_3,Mixer_3,Mixer_0,Mixer_0]}
 {id=Random_1, children [WavePlayer_5,WavePlayer_6,WavePlayer_7,WavePlayer_8], properties {weights [1,1,1,1]}}
 {id=Random_2, children [WavePlayer_9, WavePlayer_10, WavePlayer_11, WavePlayer_18], properties {weights [1,1,1,1]}}
 {id=Random_5, children [WavePlayer_19, WavePlayer_20, WavePlayer_21, WavePlayer_22], properties {weights [1,1,1,1]}}
 {id=Random_6, children [WavePlayer_23, WavePlayer_24], properties {weights [1,1]}}
 {id=Random_7, children [WavePlayer_27, WavePlayer_28], properties {weights [1,1]}}
 {id=Random_8, children [WavePlayer_31, WavePlayer_32], properties {weights [1,1]}}
 {
  id=SoundClass_1
  children [Attenuation_0]
  properties {soundClassOverride=&Fort_WeaponRanged_Fire_3P_PSM}
 }
 {
  id=SoundClass_2
  children [Attenuation_1]
  properties {soundClassOverride=&Fort_WeaponRanged_Fire_Distant}
 }
 {
  id=SoundClass_3
  children [Attenuation_11]
  properties {soundClassOverride=&Fort_WeaponRanged_Fire_Distant}
 }
 {
  id=SoundClass_4
  children [Attenuation_10]
  properties {soundClassOverride=&Fort_WeaponRanged_Fire_Distant}
 }
 {
  id=SoundClass_5
  children [Attenuation_2]
  properties {soundClassOverride=&Fort_WeaponRanged_Fire_Distant}
 }
 {
  id=SoundClass_7
  children [Attenuation_9]
  properties {soundClassOverride=&Fort_WeaponRanged_Fire_3P_PSM}
 }
 {id=WavePlayer_10, properties {soundWaveAssetPtr=&/DrunkenRocket_Fire_Medium_02}}
 {id=WavePlayer_11, properties {soundWaveAssetPtr=&/DrunkenRocket_Fire_Medium_03}}
 {id=WavePlayer_18, properties {soundWaveAssetPtr=&/DrunkenRocket_Fire_Medium_04}}
 {id=WavePlayer_19, properties {soundWaveAssetPtr=&/DrunkenRocket_Fire_Distant_01}}
 {id=WavePlayer_20, properties {soundWaveAssetPtr=&/DrunkenRocket_Fire_Distant_02}}
 {id=WavePlayer_21, properties {soundWaveAssetPtr=&/DrunkenRocket_Fire_Distant_03}}
 {id=WavePlayer_22, properties {soundWaveAssetPtr=&/DrunkenRocket_Fire_Distant_04}}
 {id=WavePlayer_23, properties {soundWaveAssetPtr=&/DrunkenRocket_Fire_Close_04}}
 {id=WavePlayer_24, properties {soundWaveAssetPtr=&/DrunkenRocket_Fire_Close_01}}
 {id=WavePlayer_27, properties {soundWaveAssetPtr=&/DrunkenRocket_Fire_Medium_01}}
 {id=WavePlayer_28, properties {soundWaveAssetPtr=&DrunkenRocket_Fire_Medium_02}}
 {id=WavePlayer_31, properties {soundWaveAssetPtr=&DrunkenRocket_Fire_Distant_01}}
 {id=WavePlayer_32, properties {soundWaveAssetPtr=&DrunkenRocket_Fire_Distant_02}}
 {id=WavePlayer_5, properties {soundWaveAssetPtr=&DrunkenRocket_Fire_Close_04}}
 {id=WavePlayer_6, properties {soundWaveAssetPtr=&DrunkenRocket_Fire_Close_01}}
 {id=WavePlayer_7, properties {soundWaveAssetPtr=&/DrunkenRocket_Fire_Close_02}}
 {id=WavePlayer_8, properties {soundWaveAssetPtr=&/DrunkenRocket_Fire_Close_03}}
 {id=WavePlayer_9, properties {soundWaveAssetPtr=&DrunkenRocket_Fire_Medium_01}}
]
properties {
 maxDistance=27500
 volumeMultiplier=0.7
 attenuationOverrides {falloffDistance=5000, attenuationShapeExtents {x=1000}}
 concurrencyOverrides {resolutionRule=PreventNew}
 concurrencySet [&Weapon_Gunfire_Local_Conc, &Weapon_Gunfire_Global_Conc]
}
