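# Sound cue for the drunken rocket's in-flight audio, pairing a random launch one-shot with a looping flight layer.
# A quality switch selects between two mixes: four start variants with a nearer-attenuated loop, or two with a distant-attenuated loop.
# Envelopers fade loops in and shape starts, pitch modulators vary loops, and each layer has its own sound class and attenuation.
# Per-owner concurrency caps instances, stopping the quietest.
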SoundCue /Game/Sounds/Fort_Weapons/Launchers/DrunkenQuad/DrunkenRocket_InAirLoop_Cue
root=QualityLevel_0
nodes [
 {
  id=Attenuation_0
  children [Enveloper_3]
  properties {
   attenuationSettings=&Fort_Weapon_Projectile_Medium_Mono_Athena_Att
   bOverrideAttenuation=true
   attenuationOverrides {falloffDistance=10000, stereoSpread=512, attenuationShapeExtents {x=2000}}
  }
 }
 {
  id=Attenuation_10
  children [Enveloper_5]
  properties {
   attenuationSettings=&Fort_Weapon_Projectile_Medium_Mono_Athena_Att
   bOverrideAttenuation=true
   attenuationOverrides {falloffDistance=10000, stereoSpread=512, attenuationShapeExtents {x=2000}}
  }
 }
 {
  id=Attenuation_11
  children [Modulator_7]
  properties {
   attenuationSettings=/Game/Sounds/Attenuation/Fort_Weapon_Rocket_Medium_Explosion_Athena_Att
   attenuationOverrides {falloffDistance=2600, attenuationShapeExtents {x=1000}}
  }
 }
 {
  id=Attenuation_3
  children [Modulator_3]
  properties {
   attenuationSettings=/Game/Sounds/Attenuation/Fort_Weapon_Rocket_Distant_Athena_Att
   attenuationOverrides {falloffDistance=2600, attenuationShapeExtents {x=1000}}
  }
 }
 {
  id=Enveloper_2
  children [Mixer_4]
  properties {
   pitchCurve {editorCurveData {keys [{time=0}]}}
   volumeCurve {editorCurveData {keys [{interpMode=RCIM_Cubic, tangentMode=RCTM_User, time=0, value=0}, {interpMode=RCIM_Cubic, tangentMode=RCTM_User, time=1, value=1}]}}
  }
 }
 {
  id=Enveloper_3
  children [Random_2]
  properties {
   pitchCurve {editorCurveData {keys [{time=0}]}}
   volumeCurve {editorCurveData {keys [
      {interpMode=RCIM_Cubic, tangentMode=RCTM_User, time=0, value=0}
      {interpMode=RCIM_Cubic, tangentMode=RCTM_User, time=0.25, value=1}
      <at 3.7367 0.7075>
      <at 6.9018 -0.0258>
     ]}}
  }
 }
 {
  id=Enveloper_4
  children [Mixer_1]
  properties {
   pitchCurve {editorCurveData {keys [{time=0}]}}
   volumeCurve {editorCurveData {keys [{interpMode=RCIM_Cubic, tangentMode=RCTM_User, time=0, value=0}, {interpMode=RCIM_Cubic, tangentMode=RCTM_User, time=1, value=1}]}}
  }
 }
 {
  id=Enveloper_5
  children [Random_1]
  properties {
   pitchCurve {editorCurveData {keys [{time=0}]}}
   volumeCurve {editorCurveData {keys [
      {interpMode=RCIM_Cubic, tangentMode=RCTM_User, time=0, value=0}
      {interpMode=RCIM_Cubic, tangentMode=RCTM_User, time=0.25, value=1}
      <at 2.0258 0.7602>
      <at 4.5151 0.0079>
     ]}}
  }
 }
 {id=Mixer_1, children [Modulator_2,Modulator_1], properties {inputVolume [0,1.2]}}
 {id=Mixer_2, children [SoundClass_4,SoundClass_3], properties {inputVolume [0.3,0.2]}}
 {id=Mixer_3, children [SoundClass_2,SoundClass_6], properties {inputVolume [0.3,0.2]}}
 {id=Mixer_4, children [null,Modulator_4], properties {inputVolume [0.6,1.1]}}
 {id=Modulator_1, children [WavePlayer_3], properties {pitchMax=1.2, pitchMin=1.1}}
 {id=Modulator_2, children [WavePlayer_0], properties {pitchMax=0.95, pitchMin=0.85}}
 {id=Modulator_3, children [Enveloper_2], properties {pitchMax=1.1}}
 {id=Modulator_4, children [WavePlayer_13], properties {pitchMax=1.2, pitchMin=1.1}}
 {id=Modulator_7, children [Enveloper_4], properties {pitchMax=1.2, pitchMin=1.2}}
 {id=QualityLevel_0, children [Mixer_3,Mixer_3,Mixer_2,Mixer_2]}
 {id=Random_1, children [WavePlayer_7, WavePlayer_8, WavePlayer_9, WavePlayer_10], properties {weights [1,1,1,1]}}
 {id=Random_2, children [WavePlayer_14, WavePlayer_15], properties {weights [1,1]}}
 {
  id=SoundClass_2
  children [Attenuation_10]
  properties {soundClassOverride=&Fort_WeaponRanged_Fire_3P_PSM}
 }
 {
  id=SoundClass_3
  children [Attenuation_3]
  properties {soundClassOverride=&Fort_WeaponRanged_Fire_Distant}
 }
 {
  id=SoundClass_4
  children [Attenuation_0]
  properties {soundClassOverride=&Fort_WeaponRanged_Fire_3P_PSM}
 }
 {
  id=SoundClass_6
  children [Attenuation_11]
  properties {soundClassOverride=&Fort_WeaponRanged_Projectile_Effects}
 }
 {id=WavePlayer_0, properties {bLooping=true, soundWaveAssetPtr=/Game/Sounds/Fort_Weapons/Launchers/RPG_InAir_NewLoop_LP2}}
 {id=WavePlayer_10, properties {soundWaveAssetPtr=&/DrunkenRocket_InAir_Start_04}}
 {id=WavePlayer_13, properties {bLooping=true, soundWaveAssetPtr=&/DrunkenRocket_InAir_Loop_01}}
 {id=WavePlayer_14, properties {soundWaveAssetPtr=&/DrunkenRocket_InAir_Start_01}}
 {id=WavePlayer_15, properties {soundWaveAssetPtr=&/DrunkenRocket_InAir_Start_02}}
 {id=WavePlayer_3, properties {bLooping=true, soundWaveAssetPtr=&DrunkenRocket_InAir_Loop_01}}
 {id=WavePlayer_7, properties {soundWaveAssetPtr=&DrunkenRocket_InAir_Start_01}}
 {id=WavePlayer_8, properties {soundWaveAssetPtr=&DrunkenRocket_InAir_Start_02}}
 {id=WavePlayer_9, properties {soundWaveAssetPtr=&/DrunkenRocket_InAir_Start_03}}
]
properties {
 bOverrideConcurrency=true
 maxDistance=15000
 soundClassObject=&Fort_WeaponRanged_Fire_3P_PSM
 volumeMultiplier=1
 attenuationOverrides {falloffDistance=5000, attenuationShapeExtents {x=1000}}
 concurrencyOverrides {bLimitToOwner=true, maxCount=3, resolutionRule=StopQuietest}
}
