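# Sound cue for the out-of-ammo sound of the DrunkenQuad rocket launcher.
# A single wave player plays one sample at reduced volume in the weapon foley sound class.
# A concurrency override caps it at two simultaneous instances, stopping the oldest first.
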SoundCue /Game/Sounds/Fort_Weapons/Launchers/DrunkenQuad/DrunkenRocket_OutofAmmo_Cue
root=WavePlayer_0
nodes [{id=WavePlayer_0, properties {soundWaveAssetPtr=&/DrunkenRocket_OutofAmmo_01}}]
properties {
 bOverrideConcurrency=true
 maxDistance=2097152
 soundClassObject=&Fort_WeaponFoley
 volumeMultiplier=0.251
 attenuationOverrides {falloffDistance=1600}
 concurrencyOverrides {maxCount=2, resolutionRule=StopOldest}
}
