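# Sound cue for the DrunkenQuad launcher's rocket reload insert, played as weapon foley with reload attenuation.
# The immediate layer mixes a fixed Drunken insert wave with a random, pitched-down Quad rocket insert, under a shared modulator.
# A second layer plays one of two random slam waves, pitched up, after a fixed 0.7-second delay.
# A concurrency override allows a single instance, stopping the farthest first.
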SoundCue /Game/Sounds/Fort_Weapons/Launchers/DrunkenQuad/DrunkenRocket_ReloadInsert_Cue
root=Mixer_0
nodes [
 {id=Delay_0, children [Modulator_0], properties {delayMax=0.7, delayMin=0.7}}
 {id=Mixer_0, children [Modulator_1,Delay_0], properties {inputVolume [0.7,0.9]}}
 {id=Mixer_1, children [Modulator_2, WavePlayer_11], properties {inputVolume [1,1]}}
 {id=Modulator_0, children [Random_3], properties {pitchMax=1.2, pitchMin=1.1, volumeMax=0.9, volumeMin=0.8}}
 {id=Modulator_1, children [Mixer_1]}
 {id=Modulator_2, children [Random_2], properties {pitchMax=0.9, pitchMin=0.8}}
 {id=Random_2, children [WavePlayer_6, WavePlayer_10], properties {weights [1,1]}}
 {id=Random_3, children [WavePlayer_7,WavePlayer_8], properties {weights [1,1]}}
 {id=WavePlayer_10, properties {soundWaveAssetPtr=/Game/Sounds/Fort_Weapons/Launchers/Quad/QuadRocket_ReloadInsert__02}}
 {id=WavePlayer_11, properties {soundWaveAssetPtr=&/DrunkenRocket_Reload_Insert_03}}
 {id=WavePlayer_6, properties {soundWaveAssetPtr=/Game/Sounds/Fort_Weapons/Launchers/Quad/QuadRocket_ReloadInsert__01}}
 {id=WavePlayer_7, properties {soundWaveAssetPtr=&/DrunkenRocket_Reload_Insert_Slam_01}}
 {id=WavePlayer_8, properties {soundWaveAssetPtr=&/DrunkenRocket_Reload_Insert_Slam_02}}
]
properties {
 attenuationSettings=&Fort_Weapon_Reload_Att
 bOverrideConcurrency=true
 maxDistance=2250
 soundClassObject=&Fort_WeaponFoley
 volumeMultiplier=0.702
 attenuationOverrides {falloffDistance=5000, attenuationShapeExtents {x=1000}}
 concurrencyOverrides {maxCount=1, resolutionRule=StopFarthestThenPreventNew}
}
