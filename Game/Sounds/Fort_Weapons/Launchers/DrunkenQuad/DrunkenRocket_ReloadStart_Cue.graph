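# Sound cue for the start of the Drunken Quad rocket launcher's reload, using weapon foley class and reload attenuation.
# An immediate layer mixes a pitched-up open sound with a reload-start sample borrowed from the standard Quad launcher.
# Randomized shell-eject and shell-impact variants follow at staggered fixed delays, with the eject pitched down, forming a timed sequence in one cue.
# A concurrency override allows only one instance at a time.
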SoundCue /Game/Sounds/Fort_Weapons/Launchers/DrunkenQuad/DrunkenRocket_ReloadStart_Cue
root=Mixer_0
nodes [
 {id=Delay_0, children [Random_4], properties {delayMax=0.5, delayMin=0.5}}
 {id=Delay_1, children [Random_5], properties {delayMax=1, delayMin=1}}
 {id=Mixer_0, children [Mixer_1,Modulator_0,Delay_1], properties {inputVolume [0.6,1.2,1.2]}}
 {id=Mixer_1, children [Modulator_1,WavePlayer_0], properties {inputVolume [0.6,1]}}
 {id=Modulator_0, children [Delay_0], properties {pitchMax=0.9, pitchMin=0.9}}
 {id=Modulator_1, children [WavePlayer_9], properties {pitchMax=1.1, pitchMin=1.1}}
 {id=Random_4, children [WavePlayer_11, WavePlayer_12, WavePlayer_13], properties {weights [1,1,1]}}
 {id=Random_5, children [WavePlayer_14, WavePlayer_15, WavePlayer_16], properties {weights [1,1,1]}}
 {id=WavePlayer_0, properties {soundWaveAssetPtr=/Game/Sounds/Fort_Weapons/Launchers/Quad/QuadRocket_ReloadStart_02}}
 {id=WavePlayer_11, properties {soundWaveAssetPtr=&/DrunkenRocket_Reload_Start_ShellEject_01}}
 {id=WavePlayer_12, properties {soundWaveAssetPtr=&/DrunkenRocket_Reload_Start_ShellEject_02}}
 {id=WavePlayer_13, properties {soundWaveAssetPtr=&/DrunkenRocket_Reload_Start_ShellEject_03}}
 {id=WavePlayer_14, properties {soundWaveAssetPtr=&/DrunkenRocket_Reload_Start_ShellImpact_01}}
 {id=WavePlayer_15, properties {soundWaveAssetPtr=&/DrunkenRocket_Reload_Start_ShellImpact_02}}
 {id=WavePlayer_16, properties {soundWaveAssetPtr=&/DrunkenRocket_Reload_Start_ShellImpact_03}}
 {id=WavePlayer_9, properties {soundWaveAssetPtr=&/DrunkenRocket_Reload_Start_Open_02}}
]
properties {
 attenuationSettings=&Fort_Weapon_Reload_Att
 bOverrideConcurrency=true
 maxDistance=2250
 soundClassObject=&Fort_WeaponFoley
 volumeMultiplier=0.702
 attenuationOverrides {falloffDistance=5000, attenuationShapeExtents {x=1000}}
 concurrencyOverrides {maxCount=1, resolutionRule=StopFarthestThenPreventNew}
}
