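# Sound cue for dropping the DrunkenQuad rocket launcher weapon, using the WaffleTruck launcher drop wave.
# A quality-level node feeds every tier into the same single wave player, so all quality settings sound identical.
# The cue plays with slightly raised pitch through weapon-drop attenuation and the loot-collect sound class, with the PickupMaximum_Oldest concurrency set limiting overlapping plays.
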SoundCue /Game/Sounds/Fort_Weapons/Launchers/DrunkenQuad/SC_DrunkenRocket_Drop
root=QualityLevel_1
nodes [
 {id=QualityLevel_1, children [WavePlayer_0,WavePlayer_0,WavePlayer_0,WavePlayer_0]}
 {id=WavePlayer_0, properties {soundWaveAssetPtr=&/WaffleTruck_Launcher_Drop_01}}
]
properties {
 attenuationSettings=&Fort_Weapon_Drop_Att
 maxDistance=3000
 pitchMultiplier=1.1
 soundClassObject=&Fort_Loot_Collect
 volumeMultiplier=1
 concurrencyOverrides {maxCount=1, resolutionRule=PreventNew}
 concurrencySet [&PickupMaximum_Oldest]
}
